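# Niagara system for the Magic Splash vehicle boost, a paintable water-like effect trailing from the car's boost sockets.
# Foam layers persistent burst meshes with looping rate-spawned ones, split by spawn group through switch inputs and masked kills.
# Trails001 streams quality-scaled ribbons backward with caustic parameters, while WaterBurst001 fires a cone of SubUV sprites.
# All emitters apply painted-variant colors and DTCarOffset; GameModeID selects foam and trail velocities.
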NiagaraSystem /VehicleCosmetics/Boosts/Boost_MagicSplash/FX/NS_Boost_MagicSplash
dependencies {
 usedDataInterfaces [
  &NiagaraDataInterfaceCurve
  &NiagaraDataInterfaceSpriteRendererInfo
  &NiagaraDataInterfaceVectorCurve
  &NiagaraDataInterfaceSkeletalMesh
 ]
 usedDynamicInputs []
 usedModules [
  &self:SystemSpawnScript.SystemScriptSource.SystemScriptGraph.NiagaraNodeAssignment_0.SetVariables_#1
  &SystemState
 ]
}
emitters {
 Foam {
  emitterUpdateScript [
   &EmitterState(
    Inactive Response=::NewEnumerator1
    Life Cycle Mode=::NewEnumerator1
    Loop Behavior=::NewEnumerator1
    Loop Duration Mode=::NewEnumerator1
    UseLoopDelay=false
   )
   &VC_FilteredSocketCount()
   &SpawnRate(SpawnGroup=1, SpawnRate=12, Use Spawn Probability=false)
   &SpawnBurst_Instantaneous(Spawn Count=2, Use Spawn Probability=false, UseLoopCountLimit=false)
  ]
  particleSpawnScript [
   &VC_PaintedVariantsOverride(
    Black=<rgba 0.06 0.06 0.06 1>
    BurntSienna=<rgba 0.4 0.1947 0.12 1>
    Cobalt=<rgba 0.162 0.216 0.81 1>
    Crimson=<rgba 0.85 0.17 0.1927 1>
    DefaultColor=<rgba 0.2 0.5333 1 1>
    ForestGreen=<rgba 0.2627 0.65 0.2275 1>
    Grey=<rgba 0.4 0.4 0.4 1>
    Lime=<rgba 0.4547 0.62 0.124 1>
    Orange=<rgba 0.85 0.3273 0.085 1>
    Pink=<rgba 0.72 0.32 0.8 1>
    Purple=<rgba 0.32 0.16 0.8 1>
    R1 Black:LinearColor=<rgba 0.16 0.16 0.16 0.055>
    R1 BurntSienna:LinearColor=<rgba 0.62 0.3856 0.2294 0.055>
    R1 Cobalt:LinearColor=<rgba 0.3 0.5917 1 0.055>
    R1 Crimson:LinearColor=<rgba 0.95 0.3325 0.3531 0.055>
    R1 DefaultColor=<rgba 0.38 0.76 0.95 0.055>
    R1 ForestGreen:LinearColor=<rgba 0.4737 0.86 0.4386 0.055>
    R1 Gold:LinearColor=<rgba 1 1 1 0.055>
    R1 Grey:LinearColor=<rgba 0.61 0.61 0.61 0.055>
    R1 Lime:LinearColor=<rgba 0.68 0.85 0.34 0.055>
    R1 Orange:LinearColor=<rgba 0.92 0.5569 0.2392 0.055>
    R1 Pink:LinearColor=<rgba 0.8052 0.4312 0.88 0.055>
    R1 Purple:LinearColor=<rgba 0.5808 0.396 0.9 0.055>
    R1 Saffron:LinearColor=<rgba 0.95 0.8392 0.285 0.055>
    R1 SkyBlue:LinearColor=<rgba 0 0.3576 0.5365 0.055>
    R1 TitaniumWhite:LinearColor=<rgba 0.9 0.9 0.9 0.055>
    R2 Black:LinearColor=<rgba 0.12 0.12 0.12 0.25>
    R2 BurntSienna:LinearColor=<rgba 0.45 0.219 0.135 0.25>
    R2 Cobalt:LinearColor=<rgba 0.1095 0.2543 0.73 0.25>
    R2 Crimson:LinearColor=<rgba 0.8 0.16 0.1813 0.25>
    R2 DefaultColor:LinearColor=<rgba 0.35 0.675 1 0.25>
    R2 ForestGreen:LinearColor=<rgba 0.4333 0.8 0.4 0.25>
    R2 Gold:LinearColor=<rgba 1 1 1 0.25>
    R2 Grey:LinearColor=<rgba 0.5 0.5 0.5 0.25>
    R2 Lime:LinearColor=<rgba 0.6423 0.82 0.287 0.25>
    R2 Orange:LinearColor=<rgba 0.8 0.3733 0.16 0.25>
    R2 Pink:LinearColor=<rgba 0.6771 0.3626 0.74 0.25>
    R2 Purple:LinearColor=<rgba 0.4477 0.2465 0.85 0.25>
    R2 Saffron:LinearColor=<rgba 0.8 0.6933 0.16 0.25>
    R2 SkyBlue:LinearColor=<rgba 0 0.3576 0.5365 0.25>
    R2 TitaniumWhite:LinearColor=<rgba 0.8 0.8 0.8 0.25>
    R3 Black:LinearColor=<rgba 0.23 0.23 0.23 0.055>
    R3 BurntSienna:LinearColor=<rgba 0.81 0.4304 0.2106 0.055>
    R3 Cobalt:LinearColor=<rgba 0.26 0.6933 1.3 0.055>
    R3 Crimson:LinearColor=<rgba 1.3 0.39 0.4658 0.055>
    R3 DefaultColor:LinearColor=<rgba 0.39 0.9967 1.3 0.055>
    R3 ForestGreen:LinearColor=<rgba 0.39 1.3 0.39 0.055>
    R3 Gold:LinearColor=<rgba 1 1 1 0.055>
    R3 Grey:LinearColor=<rgba 0.7 0.7 0.7 0.055>
    R3 Lime:LinearColor=<rgba 0.8667 1.3 0.26 0.055>
    R3 Orange:LinearColor=<rgba 1.3 0.52 0.26 0.055>
    R3 Pink:LinearColor=<rgba 1.1109 0.3654 1.26 0.055>
    R3 Purple:LinearColor=<rgba 0.7219 0.4375 1.25 0.055>
    R3 Saffron:LinearColor=<rgba 1.25 1.1042 0.375 0.055>
    R3 SkyBlue:LinearColor=<rgba 0 0.3576 0.5365 0.055>
    R3 TitaniumWhite:LinearColor=<rgba 1.2 1.2 1.2 0.055>
    Saffron=<rgba 0.8 0.6933 0.16 1>
    SkyBlue=<rgba 0 0.4 0.6 1>
    TitaniumWhite=<rgba 0.7 0.7 0.7 1>
    Use R1 Painted Overrides=true
    Use R2 Painted Overrides=true
    Use R3 Painted Overrides=true
   )
   &InitializeParticle(
    Color=$Particles.Color
    Lifetime=&MergeHelper_FloatSwitch(0=1, 1=0.8, 2=?, 3=?, Index=&SpawnGroup(), Num=::NewEnumerator0)
    Mesh Scale=&MergeHelper_VectorSwitch(
     0=&Multiply_VectorByFloat(Vector=<2 2 2.3>, Float=1)
     1=&Multiply_VectorByFloat(Vector=<1.45 1.3 1.3>, Float=0.8)
     2=?
     3=?
     Index=&SpawnGroup()
     Num=::NewEnumerator0
    )
    Mesh Scale Mode=::NewEnumerator4
    Position Offset=<-23 0 0>
    UsePositionOffset=false
    Write Material Random=true
   )
   &PartitionParticles(Number Of Partitions=$Emitter.PortCount)
   &SkeletalMeshLocation(
    Bone / Socket Index=$Particles.PartitionParticles.Partition
    Bone / Socket Sampling Mode=::NewEnumerator5
    MaskSpawnGroup=false
    Mesh Sampling Type=::NewEnumerator6
    Sampled Position Offset=&AddVectorToPosition(
     Position=&MergeHelper_VectorSwitch(0=<-50 0 0>, 1=<-75 0 0>, 2=?, 3=?, Index=&SpawnGroup(), Num=::NewEnumerator0)
     Vector=$User.DTCarOffset
    )
    Skeletal Mesh=$System.SkeletalMesh
   )
   &InitialMeshOrientation(
    Mesh Orientation Mode=::NewEnumerator7
    Rotation=&MergeHelper_VectorSwitch(
     0=<0.75 0 0>
     1=&RandomRangeVector(Minimum=<0 0 0>, Maximum=<1 0 0>)
     2=?
     3=?
     Index=&SpawnGroup()
     Num=::NewEnumerator0
    )
    Use Rotation Vector=true
   )
   &AddVelocity(Velocity=&MergeHelper_VectorSwitch(
     0=<0 0 0>
     1=&MergeHelper_VectorSwitch(0=<-5 0 0>, 1=<-4 0 0>, 2=<-4 0 0>, 3=?, Index=$User.GameModeID, Num=::NewEnumerator1)
     2=?
     3=?
     Index=&SpawnGroup()
     Num=::NewEnumerator0
    ))
   /SetVariables(Particles.VisibilityTag=$Emitter.SpawnGroup) {isSetParameters=true}
  ]
  particleUpdateScript [
   &ParticleState(Kill Particles When Lifetime Has Elapsed=false)
   &ScaleMeshSize(Scale Factor=&MergeHelper_VectorSwitch(
     0=<1 1 1>
     1=&VectorFromCurve(VectorCurve=DataInterface() {propertyValues {
        XCurve {keys [<at 0 0.2>,<at 0.3 0.8>,<at 0.55 0.8>,<at 0.9 0.1>]}
        YCurve {keys [<at 0 0.2>,<at 0.3 0.8>,<at 0.55 0.8>,<at 0.9 0.1>]}
        ZCurve {keys [
          {arriveTangent=0.0525, interpMode=RCIM_Cubic, leaveTangent=0.0525, tangentMode=RCTM_User, time=0.01, value=0}
          {arriveTangent=2.6386, tangentMode=RCTM_User, time=1, value=3}
         ]}
       }})
     2=?
     3=?
     Index=&SpawnGroup()
     Num=::NewEnumerator0
    ))
   &SolveForcesAndVelocity(Clamp Velocity=false, Limit Acceleration=false)
   &ScaleColor(
    Scale Alpha=&MergeHelper_FloatSwitch(
     0=0.12
     1=&FloatFromCurve(FloatCurve=DataInterface() {propertyValues {Curve {
         defaultValue=1
         keys [
          {arriveTangent=0.1739, interpMode=RCIM_Cubic, leaveTangent=0.1739, tangentMode=RCTM_User, time=0, value=0.2}
          {arriveTangent=-1.7226, interpMode=RCIM_Cubic, leaveTangent=-0.2353, tangentMode=RCTM_Break, time=0.5, value=0.4}
          {interpMode=RCIM_Cubic, tangentMode=RCTM_User, time=0.85, value=0}
         ]
        }}})
     2=?
     3=?
     Index=&SpawnGroup()
     Num=::NewEnumerator0
    )
    Scale RGB=&VectorFromFloat(Value=&FloatFromCurve(FloatCurve=DataInterface() {propertyValues {Curve {keys [
          {interpMode=RCIM_Cubic, leaveTangentWeight=0.3333, tangentMode=RCTM_User, time=0, value=1.5}
          {arriveTangent=-0.3517, arriveTangentWeight=0.3333, interpMode=RCIM_Cubic, leaveTangent=-0.3517, tangentMode=RCTM_User, time=1, value=0.9}
         ]}}}))
    ScaleA=true
    ScaleRGB=true
   )
   &KillParticles(Kill Particles=/Niagara/DynamicInputs/Helpers/MaskBoolBySpawnGroup(
     Masked Bool:NiagaraBool=&SetBoolByFloatComparison(A=$Particles.NormalizedAge, B=1, Comparison Type=::NewEnumerator2)
     Passthrough Bool:NiagaraBool=false
     Spawn Group Mask:NiagaraInt32=1
    ))
  ]
  rendererClasses [&NiagaraMeshRendererProperties, &NiagaraSpriteRendererProperties]
  renderers [
   {rendererClass=&NiagaraMeshRendererProperties, rendererIndex=0}
   {rendererClass=&NiagaraMeshRendererProperties, rendererIndex=1}
   {rendererClass=&NiagaraMeshRendererProperties, rendererIndex=2}
   {rendererClass=&NiagaraMeshRendererProperties, rendererIndex=3}
   {rendererClass=&NiagaraMeshRendererProperties, rendererIndex=4}
   {rendererClass=&NiagaraMeshRendererProperties, rendererIndex=5}
   {rendererClass=&NiagaraMeshRendererProperties, rendererIndex=6}
   {rendererClass=&NiagaraMeshRendererProperties, rendererIndex=7}
   {rendererClass=&NiagaraSpriteRendererProperties, rendererIndex=8}
  ]
 }
 Trails001 {
  emitterUpdateScript [
   &EmitterState()
   &VC_FilteredSocketCount()
   &VC_RateQualityHandler(Epic Spawn Rate=50, High Spawn Rate=40, Low Spawn Rate=30, Use Custom Rate?=true)
   &VC_SpawnRibbon_v1(Number of Ribbons=$Emitter.PortCount, Spawn Time=-1, SpawnRate=$Emitter.Spawn Rate)
  ]
  particleSpawnScript [
   &VC_PaintedVariantsOverride(
    Black=<rgba 0.06 0.06 0.06 4>
    BurntSienna=<rgba 0.25 0.1253 0.0675 3.5>
    Cobalt=<rgba 0.0275 0.071 0.55 3.5>
    Crimson=<rgba 0.4 0.108 0.1177 3.5>
    DefaultColor=<rgba 0.2 0.52 1 1>
    ForestGreen=<rgba 0.0907 0.34 0.068 3.5>
    Grey=<rgba 0.26 0.26 0.26 3>
    Lime=<rgba 0.2403 0.34 0.068 3.5>
    Orange=<rgba 0.5 0.2167 0.075 3.5>
    Pink=<rgba 0.3833 0.15 0.5 3.5>
    Purple=<rgba 0.2061 0.0682 0.62 3.5>
    R1 Black:LinearColor=<rgba 0.14 0.14 0.14 1>
    R1 BurntSienna:LinearColor=<rgba 0.26 0.1387 0.078 1>
    R1 Cobalt:LinearColor=<rgba 0.075 0.1812 0.5 1>
    R1 Crimson:LinearColor=<rgba 0.5 0.175 0.1858 1>
    R1 DefaultColor=<rgba 0.05 0.43 1 1>
    R1 ForestGreen:LinearColor=<rgba 0.175 0.5 0.2021 1>
    R1 Grey:LinearColor=<rgba 0.3 0.3 0.3 1>
    R1 Lime:LinearColor=<rgba 0.35 0.5 0.05 1>
    R1 Orange:LinearColor=<rgba 0.5 0.2667 0.1 1>
    R1 Pink:LinearColor=<rgba 0.5633 0.13 0.65 1>
    R1 Purple:LinearColor=<rgba 0.2383 0.0825 0.55 1>
    R1 Saffron:LinearColor=<rgba 0.8 0.68 0.08 1>
    R1 TitaniumWhite:LinearColor=<rgba 0.7 0.7 0.7 1>
    Saffron=<rgba 0.45 0.3611 0.0945 3.5>
    TitaniumWhite=<rgba 0.45 0.45 0.45 2>
    Use R1 Painted Overrides=true
   )
   &VC_InitializeRibbon_v1(
    Color=$Particles.Color
    Custom Offset=true
    Enable Skeletal Position=true
    Lifetime=0.8
    Local Vector Offset=&AddVectorToPosition(Position=<-50 0 0>, Vector=$User.DTCarOffset)
    Offset Amount=1
    Ribbon Width=90
   )
   &AddVelocity(Velocity=&MergeHelper_VectorSwitch(0=<-2000 0 0>, 1=<-1800 0 0>, 2=<-1800 0 0>, 3=?, Index=$User.GameModeID, Num=::NewEnumerator1)) {module=AddVelocity001}
  ]
  particleUpdateScript [
   &ParticleState()
   &InheritSourceMovement(
    Applied Position Scale=&Multiply_VectorByFloat(
     Vector=<1 1 1>
     Float=&FloatFromCurve(FloatCurve=DataInterface() {propertyValues {Curve {keys [<at 0.1 1 RCIM_Cubic>,<at 1 0.5 RCIM_Cubic>]}}})
    )
    ApplyForce=false
    ApplyPosition=true
    ApplyPositionFalloff=false
   )
   &ScaleRibbonWidth(Ribbon Width Scale=&FloatFromCurve(FloatCurve=DataInterface() {propertyValues {Curve {keys [
         {interpMode=RCIM_Cubic, leaveTangentWeight=0.2978, tangentMode=RCTM_User, time=0, value=1}
         {arriveTangent=-0.0472, interpMode=RCIM_Cubic, leaveTangent=-0.0472, tangentMode=RCTM_User, time=0.25, value=1.13}
         {
          arriveTangent=-0.8032
          arriveTangentWeight=0.3766
          interpMode=RCIM_Cubic
          leaveTangent=-1.4073
          leaveTangentWeight=0.0948
          tangentMode=RCTM_Break
          time=0.7989
          value=0.7832
         }
         {arriveTangent=-1.0295, arriveTangentWeight=0.067, interpMode=RCIM_Cubic, leaveTangent=-1.0295, tangentMode=RCTM_User, time=1, value=0.5}
        ]}}}))
   /SetVariables(
    Particles.CausticColor:LinearColor=&MultiplyLinearColor(
     Linear Color=$Particles.R1 Color
     Scale Factor=&MakeVector4FromVectorAndFloat("Vector (RGB)"=&VectorFromFloat(Value=2), "Float (Alpha)"=2)
    )
    Particles.CausticsWidth:NiagaraFloat=&FloatFromCurve(
     FloatCurve=DataInterface() {propertyValues {Curve {keys [
         {interpMode=RCIM_Cubic, leaveTangentWeight=0.2978, tangentMode=RCTM_User, time=0, value=1}
         {
          arriveTangent=-0.8032
          arriveTangentWeight=0.3766
          interpMode=RCIM_Cubic
          leaveTangent=-1.4073
          leaveTangentWeight=0.0948
          tangentMode=RCTM_Break
          time=0.7989
          value=0.7832
         }
         {arriveTangent=-1.0295, arriveTangentWeight=0.067, interpMode=RCIM_Cubic, leaveTangent=-1.0295, tangentMode=RCTM_User, time=1, value=0.4}
        ]}}}
     Scale Curve=40
    )
   ) {isSetParameters=true}
   &SolveForcesAndVelocity(Clamp Velocity=false, Limit Acceleration=false)
   &VC_UpdateRibbon_v1(Enable Skeletal Position=true, Sampled Position Coordinate Space:ENiagaraCoordinateSpace=::Local)
   &ScaleColor(
    Scale Alpha=&FloatFromCurve(
     FloatCurve=DataInterface() {propertyValues {Curve {keys [<at 0.2 1 RCIM_Cubic>, {arriveTangent=-1.8793, interpMode=RCIM_Cubic, leaveTangent=-1.8793, tangentMode=RCTM_User, time=1, value=0}]}}}
     Scale Curve=1.2
    )
    Scale RGB=&VectorFromFloat(Value=2)
    ScaleA=true
    ScaleRGB=true
   )
  ]
  rendererClasses [&NiagaraRibbonRendererProperties]
  renderers [
   {rendererClass=&NiagaraRibbonRendererProperties}
   {rendererClass=&NiagaraRibbonRendererProperties, rendererIndex=1}
   {rendererClass=&NiagaraRibbonRendererProperties, rendererIndex=2}
   {rendererClass=&NiagaraRibbonRendererProperties, rendererIndex=3}
  ]
 }
 WaterBurst001 {
  emitterUpdateScript [
   &EmitterState(Inactive Response=::NewEnumerator1, Life Cycle Mode=::NewEnumerator1, Loop Behavior=::NewEnumerator1, Loop Duration=1, UseLoopDelay=false)
   &VC_FilteredSocketCount()
   &SpawnRate(SpawnRate=300, Use Spawn Probability=false) {enabled=false}
   &SpawnBurst_Instantaneous(Spawn Count=16, Use Spawn Probability=false, UseLoopCountLimit=false)
  ]
  particleSpawnScript [
   &VC_PaintedVariantsOverride(
    Black=<rgba 0.06 0.06 0.06 1>
    BurntSienna=<rgba 0.4 0.1947 0.12 1>
    Cobalt=<rgba 0.162 0.216 0.81 1>
    Crimson=<rgba 0.85 0.17 0.1927 1>
    DefaultColor=<rgba 0.2 0.5333 1 1>
    ForestGreen=<rgba 0.2627 0.65 0.2275 1>
    Grey=<rgba 0.4 0.4 0.4 1>
    Lime=<rgba 0.4547 0.62 0.124 1>
    Orange=<rgba 0.85 0.3273 0.085 1>
    Pink=<rgba 0.72 0.32 0.8 1>
    Purple=<rgba 0.32 0.16 0.8 1>
    Saffron=<rgba 0.8 0.6933 0.16 1>
    SkyBlue=<rgba 0 0.4 0.6 1>
    TitaniumWhite=<rgba 0.7 0.7 0.7 1>
   )
   &InitializeParticle(
    Color=$Particles.Color
    Lifetime Max=0.5
    Lifetime Min=0.3
    Lifetime Mode=::NewEnumerator1
    Position Offset=<-20 0 0>
    Sprite Rotation Mode=::NewEnumerator2
    Sprite Size Mode=::NewEnumerator1
    Uniform Sprite Size Max=130
    Uniform Sprite Size Min=110
    UsePositionOffset=false
    Write Material Random=true
   )
   &PartitionParticles(Number Of Partitions=$Emitter.PortCount)
   &SkeletalMeshLocation(
    Bone / Socket Index=$Particles.PartitionParticles.Partition
    Bone / Socket Sampling Mode=::NewEnumerator5
    MaskSpawnGroup=false
    Mesh Sampling Type=::NewEnumerator6
    Skeletal Mesh=$System.SkeletalMesh
   )
   &ShapeLocation(
    MaskSpawnGroup=false
    Offset=&AddVectorToPosition(Position=<-45 0 0>, Vector=$User.DTCarOffset)
    Offset Coordinate Space=::Local
    Offset Mode=::NewEnumerator0
    Override Seed=false
    Sphere Radius=10
    UseOwnerScale=true
   )
   &AddVelocity(
    Cone Angle=60
    Cone Axis=<-1 0 0>
    Module.UseConeFalloff=false
    Module.UseVelDistribution=false
    Override Seed=false
    Velocity Mode=::NewEnumerator2
    Velocity Speed=&RandomRangeFloat(Minimum=300, Maximum=500)
   )
  ]
  particleUpdateScript [
   &ParticleState()
   &ScaleSpriteSize(Scale Sprite Size Mode=::NewEnumerator0)
   &Drag(
    Drag=&FloatFromCurve(FloatCurve=DataInterface() {propertyValues {Curve {
        defaultValue=1
        keys [
         {interpMode=RCIM_Cubic, tangentMode=RCTM_User, time=0, value=0}
         {arriveTangent=1.7222, interpMode=RCIM_Cubic, leaveTangent=1.7222, tangentMode=RCTM_User, time=0.6, value=2}
        ]
       }}})
    Use Linear Drag=true
    Use Rotational Drag=false
   )
   &SolveForcesAndVelocity(Clamp Velocity=false, Limit Acceleration=false)
   &SubUVAnimation(
    "Sprite Renderer "=DataInterface() {propertyValues {SpriteRenderer=&self:NiagaraEmitter_0.NiagaraSpriteRendererProperties_0}}
    UseEndFrame=false
    UseStartFrame=false
   )
   &ScaleColor(
    Scale Alpha=&FloatFromCurve(FloatCurve=DataInterface() {propertyValues {Curve {
        defaultValue=1
        keys [
         {interpMode=RCIM_Cubic, tangentMode=RCTM_User, time=0, value=0}
         {arriveTangent=1.9654, interpMode=RCIM_Cubic, tangentMode=RCTM_Break, time=0.5, value=1}
         {tangentMode=RCTM_User, time=1, value=0}
        ]
       }}})
    ScaleA=true
    ScaleRGB=false
   )
  ]
  rendererClasses [&NiagaraSpriteRendererProperties]
  renderers [{rendererClass=&NiagaraSpriteRendererProperties}]
 }
}
summary {
 systemName=NS_Boost_MagicSplash
 emitters [
  {emitterName=WaterBurst001, rendererClasses [&NiagaraSpriteRendererProperties]}
  {emitterName=Trails001, rendererClasses [&NiagaraRibbonRendererProperties]}
  {emitterName=Foam, rendererClasses [&NiagaraMeshRendererProperties, &NiagaraSpriteRendererProperties]}
 ]
 userVariables [
  User.bIsFrontendPreview {type=NiagaraBool}
  User.bIsPainted {type=NiagaraBool}
  User.ColorVariantID {type=NiagaraInt32}
  User.DTCarOffset {type=Vector3f, description="", defaultValue {struct=Vector3f, value=<0 0 0>}}
  User.GameModeID {type=NiagaraInt32}
  User.PaintedColor {type=LinearColor, description="", defaultValue {struct=LinearColor, value=<rgba 1 1 1 1>}}
  User.PaintedScalar {type=NiagaraFloat, defaultValue=1, description=""}
  User.PortCount {type=NiagaraInt32, defaultValue=2}
 ]
}
